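# Niagara system for the AstroFeather vehicle drift trail cosmetic (MP1), attached to the car skeleton and masked per wheel.
# The active ribbon emitter draws four layers from shared particles, each with curve-driven color, width and position offsets, some swaying by sine or cosine.
# A sprite emitter throws swirling particles backward, while a wheel-wrap emitter orbits sprites and meshes around wheel sockets.
# A second, three-layer ribbon emitter is disabled; frontend-preview and paint flags switch some values.
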NiagaraSystem /VehicleCosmetics/DriftTrail/DriftTrail_AstroFeather/FX/NS_DriftTrail_AstroFeather_MP1
dependencies {
 usedDataInterfaces [
  &NiagaraDataInterfaceCurve
  &NiagaraDataInterfaceArrayFloat3
  &NiagaraDataInterfaceColorCurve
  &NiagaraDataInterfaceSkeletalMesh
 ]
 usedDynamicInputs []
 usedModules [
  &self:SystemSpawnScript.SystemScriptSource.SystemScriptGraph.NiagaraNodeAssignment_0.SetVariables_#1
  &FN_UnpackVehicleFlags
  &SystemState
  /Game/Effects/Niagara/Modules/Vehicle/FN_UpdateVehicleAttribs
  /Game/Effects/Niagara/Modules/Vehicle/FN_VehicleFlags_Cheatcodes
 ]
}
emitters {
 NE_VC_DriftTrail_Ribbon_Parent001 {
  emitterUpdateScript [
   &EmitterState()
   &VC_RateQualityHandler(High Spawn Rate=60, Low Spawn Rate=60, Medium Spawn Rate=60, Use Custom Rate?=true)
   &VC_SpawnRibbon_v1(Number of Ribbons=2, SpawnRate=$Emitter.Spawn Rate)
  ]
  particleSpawnScript [
   &VC_PaintedVariantsOverride()
   &VC_InitializeRibbon_v1(
    Custom Offset=true
    Enable Skeletal Position=true
    Facing Axis:ENiagaraOrientationAxis=::YAxis
    Lifetime=0.45
    Local Vector Offset=&AddVectorToPosition(
     Position=&MakeCustomVectorFromBool(Bool=$User.bIsFrontendPreview, True Vector=<-15 0 -20>, False Vector=<100 0 -5>)
     Vector=$User.offset
    )
    Offset Amount=1
    Ribbon Width=1
    Write Material Random=false
   )
   &AddVelocity(Velocity=<-2000 0 0>) {module=AddVelocity001}
  ]
  particleUpdateScript [
   &FN_DisableTireFX(Mask Index Attribute:NiagaraInt32=$Particles.SpawnGroup, Mask Index FL=2, Mask Index FR=3, Mask Index RL=0, Mask Index RR=1)
   /SetVariables(
    Particles.R1 Color:LinearColor=&MultiplyLinearColor(
     Linear Color=<rgba 1 1 1 10>
     Scale Factor=&Vector4FromCurve(Vector4Curve=DataInterface() {propertyValues {
        AlphaCurve {defaultValue=1, keys [{interpMode=RCIM_Cubic, tangentMode=RCTM_User, time=0, value=1}, <at 0.7807 1>, <at 1 0>]}
        BlueCurve {keys [<at 0 0.0731>,<at 0.2117 0.0024>,<at 0.3507 0.0056>,<at 0.5774 0.0123>,<at 0.7082 0.248>,<at 0.7742 0.4852>,<at 1 4>]}
        GreenCurve {keys [<at 0 0.0435>,<at 0.2117 0.0012>,<at 0.3507 0.0024>,<at 0.5774 0.0052>,<at 0.7082 0.0861>,<at 0.7742 0.1651>,<at 1 3.7804>]}
        RedCurve {keys [<at 0 1.3>,<at 0.2117 0.0356>,<at 0.3507 0.0015>,<at 0.5774 0.0012>,<at 0.7082 0.0024>,<at 0.7742 0.016>,<at 1 0.3825>]}
       }})
    )
    Particles.R1Position:NiagaraPosition=&AddVectorToPosition(
     Position=$Particles.Position
     Vector=/VehicleCosmetics/FX/Niagara/Modules/VC_ConvertVectorToWorld(Vector:Vector3f=&MakeVector(
       X=&FloatFromCurve(
        FloatCurve=DataInterface() {propertyValues {Curve {defaultValue=1, keys [{tangentMode=RCTM_User, time=0, value=0}, {tangentMode=RCTM_User, time=1, value=0}]}}}
        Scale Curve=50
       )
       Y=&Multiply_Float(
        A=&Cosine(Normalized Angle=$Emitter.Age, Period=0.8, Scale=6, Bias=3)
        B=&FloatFromCurve(
         FloatCurve=DataInterface() {propertyValues {Curve {defaultValue=1, keys [{tangentMode=RCTM_User, time=0.5, value=0}, {tangentMode=RCTM_User, time=1, value=1}]}}}
         Scale Curve=0
        )
       )
       Z=&FloatFromCurve(FloatCurve=DataInterface() {propertyValues {Curve {
           defaultValue=1
           keys [{interpMode=RCIM_Cubic, tangentMode=RCTM_Break, time=0.35, value=0}, {interpMode=RCIM_Cubic, tangentMode=RCTM_Break, time=1, value=0.3}]
          }}})
      ))
    )
    Particles.R1RibbonWidth:NiagaraFloat=&FloatFromCurve(
     FloatCurve=DataInterface() {propertyValues {Curve {
        defaultValue=1
        keys [
         {interpMode=RCIM_Cubic, tangentMode=RCTM_User, time=0.0297, value=0.7581}
         {arriveTangent=14.7161, interpMode=RCIM_Cubic, leaveTangent=14.7161, time=0.0967, value=1.873}
         {interpMode=RCIM_Cubic, tangentMode=RCTM_User, time=0.2002, value=3.2659}
         {arriveTangent=-1.8372, interpMode=RCIM_Cubic, leaveTangent=-1.8372, tangentMode=RCTM_Break, time=0.3407, value=1.0929}
         {arriveTangent=2.2197, interpMode=RCIM_Cubic, leaveTangent=2.2197, time=0.4651, value=1.3958}
         {arriveTangent=3.7988, interpMode=RCIM_Cubic, leaveTangent=3.7988, time=0.5593, value=1.5781}
         {arriveTangent=-3.8005, interpMode=RCIM_Cubic, leaveTangent=-3.8005, time=0.628, value=2.0146}
         {arriveTangent=-0.807, interpMode=RCIM_Cubic, leaveTangent=-0.807, tangentMode=RCTM_Break, time=0.7125, value=0.9959}
         {arriveTangent=-5.7886, interpMode=RCIM_Cubic, leaveTangent=-5.7886, time=0.7724, value=3.2213}
         {interpMode=RCIM_Cubic, tangentMode=RCTM_Break, time=0.8708, value=0.0796}
        ]
       }}}
     Scale Curve=40
    )
    Particles.R2Color:LinearColor=&MultiplyLinearColor(
     Linear Color=<rgba 1 1 1 10>
     Scale Factor=&Vector4FromCurve(Vector4Curve=DataInterface() {propertyValues {
        AlphaCurve {defaultValue=1, keys [{interpMode=RCIM_Cubic, tangentMode=RCTM_User, time=0, value=1}, <at 1 1>]}
        BlueCurve {keys [<at 0 0.0731>,<at 0.1988 0.0024>,<at 0.3566 0.0056>,<at 0.6433 0.0123>,<at 0.742 0.248>,<at 0.7984 0.4621>,<at 0.9902 4>]}
        GreenCurve {keys [<at 0 0.0435>,<at 0.1988 0.0012>,<at 0.3566 0.0024>,<at 0.6433 0.0052>,<at 0.742 0.0861>,<at 0.7984 0.1529>,<at 0.9902 3.7076>]}
        RedCurve {keys [<at 0 1.3>,<at 0.1988 0.0356>,<at 0.3566 0.0015>,<at 0.6433 0.0012>,<at 0.742 0.0024>,<at 0.7984 0.0152>,<at 0.9902 0.372>]}
       }})
    )
    Particles.R2Position:NiagaraPosition=&AddVectorToPosition(
     Position=$Particles.Position
     Vector=&VC_ConvertVectorToWorld(Vector:Vector3f=&MakeVector(
       X=&FloatFromCurve(
        FloatCurve=DataInterface() {propertyValues {Curve {defaultValue=1, keys [{tangentMode=RCTM_User, time=0.45, value=0}, {tangentMode=RCTM_User, time=0.6219, value=1.6}]}}}
        Scale Curve=32
       )
       Y=&Multiply_Float(
        A=&Cosine(Normalized Angle=$Emitter.Age, Period=0.5, Scale=-5, Bias=1)
        B=&FloatFromCurve(
         FloatCurve=DataInterface() {propertyValues {Curve {
            defaultValue=1
            keys [{interpMode=RCIM_Cubic, tangentMode=RCTM_Break, time=0.5, value=0}, {interpMode=RCIM_Cubic, tangentMode=RCTM_Break, time=1, value=1}]
           }}}
         Scale Curve=0
        )
       )
       Z=&FloatFromCurve(
        FloatCurve=DataInterface() {propertyValues {Curve {
           defaultValue=1
           keys [{interpMode=RCIM_Cubic, tangentMode=RCTM_Break, time=0.1, value=0}, {interpMode=RCIM_Cubic, tangentMode=RCTM_Break, time=1, value=0.9}]
          }}}
        Scale Curve=26
       )
      ))
    )
    Particles.R2RibbonWidth:NiagaraFloat=&FloatFromCurve(
     FloatCurve=DataInterface() {propertyValues {Curve {
        defaultValue=1
        keys [
         {interpMode=RCIM_Cubic, tangentMode=RCTM_User, time=0.0107, value=0.2142}
         {interpMode=RCIM_Cubic, tangentMode=RCTM_User, time=0.1328, value=1.4959}
         {arriveTangent=-1.8372, interpMode=RCIM_Cubic, leaveTangent=-1.8372, tangentMode=RCTM_Break, time=0.2255, value=2.9295}
         {arriveTangent=-0.6864, interpMode=RCIM_Cubic, leaveTangent=-0.6864, time=0.531, value=1.7958}
         {arriveTangent=-4.5468, interpMode=RCIM_Cubic, leaveTangent=-4.5468, time=0.5897, value=2.6795}
         {arriveTangent=-0.807, interpMode=RCIM_Cubic, leaveTangent=-0.807, tangentMode=RCTM_Break, time=0.7296, value=0.8928}
         {arriveTangent=-5.1904, interpMode=RCIM_Cubic, leaveTangent=-5.1904, time=0.7627, value=3.4855}
         {interpMode=RCIM_Cubic, tangentMode=RCTM_Break, time=0.8732, value=0.1476}
        ]
       }}}
     Scale Curve=40
    )
    Particles.R3Color:LinearColor=&MultiplyLinearColor(
     Linear Color=<rgba 1 1 1 10>
     Scale Factor=&Vector4FromCurve(Vector4Curve=DataInterface() {propertyValues {
        AlphaCurve {defaultValue=1, keys [{interpMode=RCIM_Cubic, tangentMode=RCTM_User, time=0, value=1}, <at 1 1>]}
        BlueCurve {keys [<at 0 0.0731>,<at 0.2295 0.0024>,<at 0.5687 0.0056>,<at 0.6731 0.0123>,<at 0.7887 0.2484>,<at 1 5>]}
        GreenCurve {keys [<at 0 0.0435>,<at 0.2295 0.0012>,<at 0.5687 0.0024>,<at 0.6731 0.0052>,<at 0.7887 0.0863>,<at 1 4.7255>]}
        RedCurve {keys [<at 0 1.3>,<at 0.2295 0.0356>,<at 0.5687 0.0015>,<at 0.6731 0.0012>,<at 0.7887 0.0024>,<at 1 0.4781>]}
       }})
    )
    Particles.R3Position:NiagaraPosition=&AddVectorToPosition(
     Position=$Particles.Position
     Vector=&VC_ConvertVectorToWorld(Vector:Vector3f=&MakeVector(
       X=&FloatFromCurve(
        FloatCurve=DataInterface() {propertyValues {Curve {defaultValue=1, keys [{tangentMode=RCTM_User, time=0.15, value=0}, {tangentMode=RCTM_User, time=0.95, value=1.7}]}}}
        Scale Curve=90
       )
       Y=&Multiply_Float(
        A=&Sine(Normalized Angle=$Emitter.Age, Period=0.9, Scale=5, Bias=1)
        B=&FloatFromCurve(
         FloatCurve=DataInterface() {propertyValues {Curve {defaultValue=1, keys [{tangentMode=RCTM_User, time=0.6, value=0}, {tangentMode=RCTM_User, time=1, value=1}]}}}
         Scale Curve=7
        )
       )
       Z=&FloatFromCurve(
        FloatCurve=DataInterface() {propertyValues {Curve {defaultValue=1, keys [{tangentMode=RCTM_User, time=0.3, value=0}, {tangentMode=RCTM_User, time=1, value=1.2}]}}}
        Scale Curve=40
       )
      ))
    )
    Particles.R3RibbonWidth:NiagaraFloat=&FloatFromCurve(
     FloatCurve=DataInterface() {propertyValues {Curve {
        defaultValue=1
        keys [
         {interpMode=RCIM_Cubic, tangentMode=RCTM_User, time=0.0107, value=0.2142}
         {interpMode=RCIM_Cubic, tangentMode=RCTM_User, time=0.1328, value=1.3783}
         {arriveTangent=-1.8372, interpMode=RCIM_Cubic, leaveTangent=-1.8372, tangentMode=RCTM_Break, time=0.2389, value=2.3122}
         {arriveTangent=0.0855, interpMode=RCIM_Cubic, leaveTangent=0.0855, time=0.4046, value=0.9084}
         {arriveTangent=1.7929, interpMode=RCIM_Cubic, leaveTangent=1.7929, time=0.5565, value=2.3393}
         {arriveTangent=-0.807, interpMode=RCIM_Cubic, leaveTangent=-0.807, tangentMode=RCTM_Break, time=0.6456, value=1.3404}
         {arriveTangent=-6.366, interpMode=RCIM_Cubic, leaveTangent=-6.366, time=0.7525, value=3.5027}
         {interpMode=RCIM_Cubic, tangentMode=RCTM_Break, time=0.8654, value=-0.0588}
        ]
       }}}
     Scale Curve=40
    )
    Particles.R4Color:LinearColor=&MultiplyLinearColor(
     Linear Color=<rgba 1 1 1 10>
     Scale Factor=&Vector4FromCurve(Vector4Curve=DataInterface() {propertyValues {
        AlphaCurve {defaultValue=1, keys [{interpMode=RCIM_Cubic, tangentMode=RCTM_User, time=0.0054, value=1}, <at 0.91 0.5756>, <at 0.9822 0>]}
        BlueCurve {keys [<at 0 0.0186>,<at 0.7827 0.0419>,<at 0.9245 0.4852>,<at 1 2.2>]}
        GreenCurve {keys [<at 0 0.0076>,<at 0.7827 0.0155>,<at 0.9245 0.1651>,<at 1 2.0792>]}
        RedCurve {keys [<at 0 0.0018>,<at 0.7827 0.0025>,<at 0.9245 0.016>,<at 1 0.2104>]}
       }})
    )
    Particles.R4Position:NiagaraPosition=&AddVectorToPosition(
     Position=$Particles.Position
     Vector=&VC_ConvertVectorToWorld(Vector:Vector3f=&MakeVector(
       X=&FloatFromCurve(
        FloatCurve=DataInterface() {propertyValues {Curve {defaultValue=1, keys [{tangentMode=RCTM_User, time=0, value=0}, {tangentMode=RCTM_User, time=1, value=1}]}}}
        Scale Curve=0
       )
       Y=&Multiply_Float(
        A=&Cosine(Normalized Angle=$Emitter.Age, Period=0.8, Scale=6, Bias=3)
        B=&FloatFromCurve(
         FloatCurve=DataInterface() {propertyValues {Curve {defaultValue=1, keys [{tangentMode=RCTM_User, time=0.5, value=0}, {tangentMode=RCTM_User, time=1, value=1}]}}}
         Scale Curve=5
        )
       )
       Z=&FloatFromCurve(
        FloatCurve=DataInterface() {propertyValues {Curve {
           defaultValue=1
           keys [{interpMode=RCIM_Cubic, tangentMode=RCTM_Break, time=0.35, value=0}, {interpMode=RCIM_Cubic, tangentMode=RCTM_Break, time=1, value=1.1}]
          }}}
        Scale Curve=45
       )
      ))
    )
    Particles.R4SmokeWidth:NiagaraFloat=&FloatFromCurve(
     FloatCurve=DataInterface() {propertyValues {Curve {
        defaultValue=1
        keys [
         {interpMode=RCIM_Cubic, tangentMode=RCTM_User, time=0.0165, value=0}
         {arriveTangent=-0.0038, interpMode=RCIM_Cubic, leaveTangent=-0.0038, tangentMode=RCTM_User, time=0.3, value=0.1238}
         {interpMode=RCIM_Cubic, tangentMode=RCTM_User, time=0.4911, value=1.3632}
         {tangentMode=RCTM_User, time=1.011, value=1.4407}
        ]
       }}}
     Scale Curve=50
    )
   ) {isSetParameters=true}
   &InheritSourceMovement(
    Applied Position Scale=&VectorFromFloat(Value=&FloatFromCurve(FloatCurve=DataInterface() {propertyValues {Curve {keys [
          {interpMode=RCIM_Cubic, tangentMode=RCTM_Break, time=0, value=1}
          {arriveTangent=-2.9025, interpMode=RCIM_Cubic, tangentMode=RCTM_Break, time=1, value=0}
         ]}}}))
    ApplyForce=false
    ApplyPosition=true
    ApplyPositionFalloff=false
   )
   &DynamicMaterialParameters(
    Index 3 Param 1=&FloatFromCurve(FloatCurve=DataInterface() {propertyValues {Curve {keys [
         {interpMode=RCIM_Cubic, leaveTangent=5.8099, tangentMode=RCTM_Break, time=0, value=-55}
         {arriveTangent=77.992, interpMode=RCIM_Cubic, leaveTangent=77.992, tangentMode=RCTM_User, time=0.1037, value=-52.6475}
         {arriveTangent=1.6276, interpMode=RCIM_Cubic, tangentMode=RCTM_Break, time=0.5, value=20}
        ]}}})
    Module.Index0Float0Write=true
    Module.Index0Float1Write=true
    Module.Index0Float2Write=true
    Module.Index0Float3Write=true
    Module.Index3Float0Write=true
    Module.Index3Float1Write=true
    Module.Index3Float2Write=true
    Module.Index3Float3Write=true
    Write Parameter Index 3=::NewEnumerator1
   )
   &ParticleState()
   &SolveForcesAndVelocity(Clamp Velocity=false, Limit Acceleration=false)
   &VC_UpdateRibbon_v1(Enable Skeletal Position=true)
  ]
  rendererClasses [&NiagaraRibbonRendererProperties]
  renderers [
   {rendererClass=&NiagaraRibbonRendererProperties}
   {rendererClass=&NiagaraRibbonRendererProperties, rendererIndex=1}
   {rendererClass=&NiagaraRibbonRendererProperties, rendererIndex=2}
   {rendererClass=&NiagaraRibbonRendererProperties, rendererIndex=3}
  ]
 }
 NE_VC_DriftTrail_Ribbon_Parent003 {
  bEnabled=false
  emitterUpdateScript [
   &EmitterState()
   &VC_RateQualityHandler(High Spawn Rate=60, Low Spawn Rate=60, Medium Spawn Rate=60, Use Custom Rate?=true)
   &VC_SpawnRibbon_v1(Number of Ribbons=2, SpawnRate=$Emitter.Spawn Rate)
  ]
  particleSpawnScript [
   &VC_PaintedVariantsOverride() {enabled=false}
   /SetVariables(Particles.R2Lifetime=0.3) {isSetParameters=true}
   &VC_InitializeRibbon_v1(
    Custom Offset=true
    Enable Skeletal Position=true
    Facing Axis:ENiagaraOrientationAxis=::YAxis
    Lifetime=0.6
    Local Vector Offset=&AddVectorToPosition(
     Position=&MakeCustomVectorFromBool(
      Bool=$User.bIsFrontendPreview
      True Vector=&SelectVectorFromArray(
       Array Sampling Mode=::NewEnumerator0
       Direct Array Index=?
       Vector Selection Array=DataInterface() {propertyValues {FloatData [<50 0 -10>, <50 1 -10>]}}
       Random Seed=0
      )
      False Vector=<100 0 -10>
     )
     Vector=$User.offset
    )
    Offset Amount=1
    Ribbon Width=1
    Write Material Random=false
   )
   &AddVelocity(Velocity=<-2500 0 0>) {module=AddVelocity001}
  ]
  particleUpdateScript [
   &FN_DisableTireFX(Mask Index Attribute:NiagaraInt32=$Particles.SpawnGroup, Mask Index FL=2, Mask Index FR=3, Mask Index RL=0, Mask Index RR=1)
   /SetVariables(
    Particles.R1 Color:LinearColor=&MultiplyLinearColor(
     Linear Color=<rgba 1 1 1 10>
     Scale Factor=&Vector4FromCurve(Vector4Curve=DataInterface() {propertyValues {
        AlphaCurve {defaultValue=1, keys [{interpMode=RCIM_Cubic, tangentMode=RCTM_User, time=0, value=1}, <at 1 1>]}
        BlueCurve {keys [<at 0 0.0731>,<at 0.2368 0.0024>,<at 0.5037 0.0056>,<at 0.6155 0.0056>,<at 0.7181 0.0123>,<at 0.8082 0.248>,<at 1 2>]}
        GreenCurve {keys [<at 0 0.0435>,<at 0.2368 0.0012>,<at 0.5037 0.0024>,<at 0.6155 0.0024>,<at 0.7181 0.0052>,<at 0.8082 0.0861>,<at 1 1.8902>]}
        RedCurve {keys [<at 0 1.3>,<at 0.2368 0.0356>,<at 0.5037 0.0015>,<at 0.6155 0.0015>,<at 0.7181 0.0012>,<at 0.8082 0.0024>,<at 1 0.1912>]}
       }})
    )
    Particles.R1RibbonWidth:NiagaraFloat=&FloatFromCurve(
     FloatCurve=DataInterface() {propertyValues {Curve {
        defaultValue=1
        keys [
         {interpMode=RCIM_Cubic, tangentMode=RCTM_User, time=0.0297, value=0.7581}
         {arriveTangent=14.7161, interpMode=RCIM_Cubic, leaveTangent=14.7161, time=0.0967, value=1.873}
         {interpMode=RCIM_Cubic, tangentMode=RCTM_User, time=0.2002, value=3.2659}
         {arriveTangent=-1.8372, interpMode=RCIM_Cubic, leaveTangent=-1.8372, tangentMode=RCTM_Break, time=0.3178, value=1.7504}
         {arriveTangent=-1.0669, interpMode=RCIM_Cubic, leaveTangent=-1.0669, time=0.4753, value=2.3905}
         {arriveTangent=-0.807, interpMode=RCIM_Cubic, leaveTangent=-0.807, tangentMode=RCTM_Break, time=0.7688, value=1.2692}
         {arriveTangent=-2.2522, interpMode=RCIM_Cubic, leaveTangent=-2.2522, time=0.8037, value=3.0432}
         {interpMode=RCIM_Cubic, tangentMode=RCTM_Break, time=0.9431, value=0.8766}
        ]
       }}}
     Scale Curve=40
    )
    Particles.R2Color:LinearColor=&MultiplyLinearColor(
     Linear Color=<rgba 1 1 1 10>
     Scale Factor=&Vector4FromCurve(Vector4Curve=DataInterface() {propertyValues {
        AlphaCurve {defaultValue=1, keys [{interpMode=RCIM_Cubic, tangentMode=RCTM_User, time=0, value=1}, <at 1 1>]}
        BlueCurve {keys [<at 0 0.0731>,<at 0.3036 0.0024>,<at 0.5037 0.0056>,<at 0.5953 0.0056>,<at 0.7041 0.0123>,<at 0.8082 0.248>,<at 1 2>]}
        GreenCurve {keys [<at 0 0.0435>,<at 0.3036 0.0012>,<at 0.5037 0.0024>,<at 0.5953 0.0024>,<at 0.7041 0.0052>,<at 0.8082 0.0861>,<at 1 1.8902>]}
        RedCurve {keys [<at 0 1.3>,<at 0.3036 0.0356>,<at 0.5037 0.0015>,<at 0.5953 0.0015>,<at 0.7041 0.0012>,<at 0.8082 0.0024>,<at 1 0.1912>]}
       }})
    )
    Particles.R2Position:NiagaraPosition=&AddVectorToPosition(
     Position=$Particles.Position
     Vector=&VC_ConvertVectorToWorld(Vector:Vector3f=&MakeVector(
       X=&FloatFromCurve(
        FloatCurve=DataInterface() {propertyValues {Curve {defaultValue=1, keys [{tangentMode=RCTM_User, time=0.4, value=0}, {tangentMode=RCTM_User, time=1, value=1}]}}}
        Scale Curve=50
       )
       Y=0
       Z=&FloatFromCurve(
        FloatCurve=DataInterface() {propertyValues {Curve {
           defaultValue=1
           keys [{interpMode=RCIM_Cubic, tangentMode=RCTM_Break, time=0.35, value=0}, {interpMode=RCIM_Cubic, tangentMode=RCTM_Break, time=1, value=1.1}]
          }}}
        Scale Curve=45
       )
      ))
    )
    Particles.R2RibbonWidth:NiagaraFloat=&FloatFromCurve(
     FloatCurve=DataInterface() {propertyValues {Curve {
        defaultValue=1
        keys [
         {interpMode=RCIM_Cubic, tangentMode=RCTM_User, time=0.0107, value=0.2142}
         {interpMode=RCIM_Cubic, tangentMode=RCTM_User, time=0.1328, value=1.4959}
         {arriveTangent=-1.8372, interpMode=RCIM_Cubic, leaveTangent=-1.8372, tangentMode=RCTM_Break, time=0.2255, value=2.9295}
         {arriveTangent=-1.9869, interpMode=RCIM_Cubic, leaveTangent=-1.9869, time=0.4073, value=1.0848}
         {arriveTangent=-0.596, interpMode=RCIM_Cubic, leaveTangent=-0.596, time=0.5966, value=2.1923}
         {arriveTangent=-0.807, interpMode=RCIM_Cubic, leaveTangent=-0.807, tangentMode=RCTM_Break, time=0.7296, value=0.8928}
         {arriveTangent=-1.4696, interpMode=RCIM_Cubic, leaveTangent=-1.4696, time=0.8146, value=2.667}
         {interpMode=RCIM_Cubic, tangentMode=RCTM_Break, time=0.9592, value=0.5553}
        ]
       }}}
     Scale Curve=40
    )
    Particles.R3Color:LinearColor=&MultiplyLinearColor(
     Linear Color=<rgba 1 1 1 10>
     Scale Factor=&Vector4FromCurve(Vector4Curve=DataInterface() {propertyValues {
        AlphaCurve {defaultValue=1, keys [{interpMode=RCIM_Cubic, tangentMode=RCTM_User, time=0, value=1}, <at 1 1>]}
        BlueCurve {keys [<at 0 0.0731>,<at 0.2865 0.0024>,<at 0.5037 0.0056>,<at 0.6218 0.0056>,<at 0.7165 0.0123>,<at 0.7989 0.2484>,<at 1 2>]}
        GreenCurve {keys [<at 0 0.0435>,<at 0.2865 0.0012>,<at 0.5037 0.0024>,<at 0.6218 0.0024>,<at 0.7165 0.0052>,<at 0.7989 0.0863>,<at 1 1.8902>]}
        RedCurve {keys [<at 0 1.3>,<at 0.2865 0.0356>,<at 0.5037 0.0015>,<at 0.6218 0.0015>,<at 0.7165 0.0012>,<at 0.7989 0.0024>,<at 1 0.1912>]}
       }})
    )
    Particles.R3Position:NiagaraPosition=&AddVectorToPosition(
     Position=$Particles.Position
     Vector=&VC_ConvertVectorToWorld(Vector:Vector3f=&MakeVector(
       X=&FloatFromCurve(
        FloatCurve=DataInterface() {propertyValues {Curve {defaultValue=1, keys [{tangentMode=RCTM_User, time=0.5, value=0}, {tangentMode=RCTM_User, time=1, value=1.3}]}}}
        Scale Curve=120
       )
       Y=0
       Z=&FloatFromCurve(
        FloatCurve=DataInterface() {propertyValues {Curve {defaultValue=1, keys [{tangentMode=RCTM_User, time=0.4, value=0}, {tangentMode=RCTM_User, time=1, value=1.2}]}}}
        Scale Curve=95
       )
      ))
    )
    Particles.R3RibbonWidth:NiagaraFloat=&FloatFromCurve(
     FloatCurve=DataInterface() {propertyValues {Curve {
        defaultValue=1
        keys [
         {interpMode=RCIM_Cubic, tangentMode=RCTM_User, time=0.0107, value=0.2142}
         {interpMode=RCIM_Cubic, tangentMode=RCTM_User, time=0.1328, value=1.3783}
         {arriveTangent=-1.8372, interpMode=RCIM_Cubic, leaveTangent=-1.8372, tangentMode=RCTM_Break, time=0.2389, value=2.3122}
         {arriveTangent=0.0855, interpMode=RCIM_Cubic, leaveTangent=0.0855, time=0.4046, value=0.9084}
         {arriveTangent=1.5146, interpMode=RCIM_Cubic, leaveTangent=1.5146, time=0.5565, value=2.3393}
         {arriveTangent=-0.807, interpMode=RCIM_Cubic, leaveTangent=-0.807, tangentMode=RCTM_Break, time=0.741, value=1.4179}
         {arriveTangent=-1.6062, interpMode=RCIM_Cubic, leaveTangent=-1.6062, time=0.8876, value=2.5939}
         {interpMode=RCIM_Cubic, tangentMode=RCTM_Break, time=0.9824, value=1.0302}
        ]
       }}}
     Scale Curve=40
    )
   ) {isSetParameters=true}
   &ParticleState()
   &InheritSourceMovement(
    Applied Position Scale=&VectorFromFloat(Value=&FloatFromCurve(FloatCurve=DataInterface() {propertyValues {Curve {keys [
          {interpMode=RCIM_Cubic, tangentMode=RCTM_Break, time=0, value=1}
          {arriveTangent=-2.9025, interpMode=RCIM_Cubic, tangentMode=RCTM_Break, time=1, value=0}
         ]}}}))
    ApplyForce=false
    ApplyPosition=true
    ApplyPositionFalloff=false
   )
   &SolveForcesAndVelocity(Clamp Velocity=false, Limit Acceleration=false)
   &VC_UpdateRibbon_v1(Enable Skeletal Position=true)
   &DynamicMaterialParameters(
    Index 3 Param 1=&FloatFromCurve(FloatCurve=DataInterface() {propertyValues {Curve {keys [
         {interpMode=RCIM_Cubic, leaveTangent=5.8099, tangentMode=RCTM_Break, time=0, value=-55}
         {arriveTangent=77.992, interpMode=RCIM_Cubic, leaveTangent=77.992, tangentMode=RCTM_User, time=0.1037, value=-52.6475}
         {arriveTangent=1.6276, interpMode=RCIM_Cubic, tangentMode=RCTM_Break, time=0.5, value=20}
        ]}}})
    Module.Index0Float0Write=true
    Module.Index0Float1Write=true
    Module.Index0Float2Write=true
    Module.Index0Float3Write=true
    Module.Index3Float0Write=true
    Module.Index3Float1Write=true
    Module.Index3Float2Write=true
    Module.Index3Float3Write=true
    Write Parameter Index 3=::NewEnumerator1
   )
  ]
  rendererClasses [&NiagaraRibbonRendererProperties]
  renderers [
   {rendererClass=&NiagaraRibbonRendererProperties}
   {rendererClass=&NiagaraRibbonRendererProperties, rendererIndex=1}
   {rendererClass=&NiagaraRibbonRendererProperties, rendererIndex=2}
  ]
 }
 NE_VC_DriftTrail_Sprite_Parent {
  emitterUpdateScript [
   &EmitterState()
   &SpawnRate(SpawnRate=52, Use Spawn Probability=false)
  ]
  particleSpawnScript [
   &VC_PaintedVariantsOverride(DefaultColor=<rgba 0.015 0.015 0.015 1>) {enabled=false}
   &InitializeParticle(
    Color=<rgba 0.0101 0.8478 0.9 1>
    Lifetime Max=0.46
    Lifetime Min=0.24
    Lifetime Mode=::NewEnumerator1
    Mesh Scale Mode=::NewEnumerator3
    Mesh Uniform Scale=&MakeCustomFloatFromBool(
     Bool=$User.bIsFrontendPreview
     True Float=&RandomRangeFloat(Minimum=0.8, Maximum=1.5)
     False Float=&RandomRangeFloat(Minimum=1.2, Maximum=2.2)
    )
    Sprite Rotation Mode=::NewEnumerator2
    Sprite Size Mode=::NewEnumerator1
    Uniform Sprite Size Max=6
    Uniform Sprite Size Min=10
    UsePositionOffset=false
    Write Material Random=true
   )
   &PartitionParticles()
   &SkeletalMeshLocation(
    Bone / Socket Index=&Add_Integer(A=$Particles.PartitionParticles.Partition, B=4)
    Bone / Socket Sampling Mode=::NewEnumerator5
    Interpolate Mesh Sampling=false
    MaskSpawnGroup=false
    Mesh Sampling Type=::NewEnumerator6
    Sampled Position Offset=&AddVectorToPosition(Position=$User.DTCarOffset, Vector=<0 0 -1.4333>)
    Skeletal Mesh=$System.SkeletalMesh
   )
   &AddVelocity(Velocity=&RandomRangeVector(Minimum=<-850 0 0>, Maximum=<-1200 0 0>))
   &ShapeLocation(
    MaskSpawnGroup=true
    Offset=<0 0 -10>
    Offset Mode=::NewEnumerator0
    Override Seed=false
    Sphere Radius=28
    Sphere Surface Distribution=1
    UseOwnerScale=true
   )
   &InitialMeshOrientation(
    Mesh Orientation Mode=::NewEnumerator2
    Rotation=&RandomRangeVector(Minimum=<0 0 0>, Maximum=<0.5 0 1>)
    Use Rotation Vector=true
   )
  ]
  particleUpdateScript [
   &InheritSourceMovement(
    Applied Position Scale=&VectorFromFloat(Value=&FloatFromCurve(FloatCurve=DataInterface() {propertyValues {Curve {keys [
          {interpMode=RCIM_Cubic, tangentMode=RCTM_Break, time=0, value=1}
          {arriveTangent=-2.9025, interpMode=RCIM_Cubic, tangentMode=RCTM_Break, time=1, value=0}
         ]}}}))
    ApplyForce=false
    ApplyPosition=true
    ApplyPositionFalloff=false
   )
   &FN_DisableTireFX(
    Mask Index Attribute:NiagaraInt32=$Particles.PartitionParticles.Partition
    Mask Index FL=2
    Mask Index FR=3
    Mask Index RL=0
    Mask Index RR=1
   )
   &CurlNoiseForce(Bias Noise=false, Noise Frequency=100, Noise Strength=50, Pan Noise=false)
   &Drag(Drag=0.4, Use Linear Drag=true, Use Rotational Drag=true)
   &VortexForce(
    Origin Pull Amount=150
    UseOriginOffset=false
    Vortex Axis=&RandomRangeVector(Minimum=<0 -20 -0.05>, Maximum=<0 0.1 0.5>)
    Vortex Force Amount=80
    Vortex Origin:NiagaraPosition=$Particles.Initial.Position
   )
   &ParticleState()
   &ScaleColor(
    Scale Alpha=&FloatFromCurve(FloatCurve=DataInterface() {propertyValues {Curve {keys [{interpMode=RCIM_Cubic, tangentMode=RCTM_Break, time=0.6789, value=1}, {interpMode=RCIM_Cubic, tangentMode=RCTM_Break, time=0.9, value=0}]}}})
    ScaleA=true
    ScaleRGB=true
   )
   &SolveForcesAndVelocity(Clamp Velocity=false, Limit Acceleration=false)
  ]
  rendererClasses [&NiagaraSpriteRendererProperties]
  renderers [{rendererClass=&NiagaraSpriteRendererProperties}]
 }
 NE_VC_DriftTrail_WheelWrap_Parent001 {
  emitterUpdateScript [
   &EmitterState(
    Inactive Response=::NewEnumerator1
    Life Cycle Mode=::NewEnumerator1
    Loop Behavior=::NewEnumerator1
    Loop Duration Mode=::NewEnumerator1
    UseLoopDelay=false
   )
   &SpawnRate(SpawnRate=70, Use Spawn Probability=false)
   &SpawnBurst_Instantaneous(Spawn Count=40, Use Spawn Probability=false, UseLoopCountLimit=false) {enabled=false}
  ]
  particleSpawnScript [
   &VC_PaintedVariantsOverride() {enabled=false}
   &InitializeParticle(
    Color Mode=::NewEnumerator0
    Lifetime=0.8
    Mesh Scale Mode=::NewEnumerator4
    Position Offset=<0 250 0>
    Sprite Rotation Mode=::NewEnumerator2
    Sprite Size Mode=::NewEnumerator1
    Uniform Sprite Size Max=6
    Uniform Sprite Size Min=3
    UsePositionOffset=false
    Write Material Random=true
   )
   &PartitionParticles()
   &SkeletalMeshLocation(
    Bone / Socket Index=&Add_Integer(A=$Particles.PartitionParticles.Partition, B=4)
    Bone / Socket Sampling Mode=::NewEnumerator5
    Interpolate Mesh Sampling=false
    MaskSpawnGroup=false
    Mesh Sampling Type=::NewEnumerator6
    Sampled Position Offset=&RandomRangeVector(Minimum=<0 -15 0>, Maximum=<0 15 0>)
    Skeletal Mesh=$System.SkeletalMesh
   )
  ]
  particleUpdateScript [
   &FN_DisableTireFX(
    Mask Index Attribute:NiagaraInt32=$Particles.PartitionParticles.Partition
    Mask Index FL=2
    Mask Index FR=3
    Mask Index RL=0
    Mask Index RR=1
   )
   &VC_WheelWrapScale(Wheel Index Attribute:NiagaraInt32=&Add_Integer(A=$Particles.PartitionParticles.Partition, B=2))
   /SetVariables(Particles.MeshColor:LinearColor=<rgba 0.9466 0.0317 0.0532 1>) {isSetParameters=true}
   &ParticleState()
   &ScaleColor(Scale RGB=&MakeVectorFromLinearColorRGB(Color=<rgba 0.1406 1.891 2 1>), ScaleA=true, ScaleRGB=true)
   &DynamicMaterialParameters(
    Index 0 Param 2=0.3
    Index 0 Param 3=&FloatFromCurve(FloatCurve=DataInterface() {propertyValues {Curve {keys [{interpMode=RCIM_Cubic, tangentMode=RCTM_User, time=0, value=0}, {interpMode=RCIM_Cubic, tangentMode=RCTM_User, time=0.1, value=1}]}}})
    Module.Index0Param0Write=true
    Module.Index0Param1Write=true
    Module.Index0Param2Write=true
    Module.Index0Param3Write=true
   )
   &RotateAroundPoint(
    Radius:NiagaraFloat=&Multiply_Float(
     A=&RandomRangeFloat(Minimum=54, Maximum=50)
     B=&MakeFloatFromVector(Vector=$Particles.Scale, Channel=::NewEnumerator0)
    )
    Rotation Center:NiagaraPosition=&MaskVectorBySpawnGroup(
     Masked Vector=&AddVectorToPosition(Position=$StackContext.Position, Vector=&RandomRangeVector(Minimum=<-3 -10 -3>, Maximum=<3 10 3>))
     Passthrough Vector=Linked() {linkedVariable {name=StackContext.Position, type=NiagaraPosition}}
    )
    Rotation Rate:NiagaraFloat=&MaskFloatBySpawnGroup(Masked Float=&RandomRangeFloat(Minimum=-1, Maximum=-0.5), Passthrough Float=0)
    Set Particle Position=false
    Y Axis:Vector3f=<0 0 1>
   ) {module=RotateAroundPoint001}
   &UpdateMeshOrientation(Rotation Coordinate Space=::NewEnumerator2, Rotation Vector=<0 1 0>)
   /SetVariables(Particles.AroundPosition:NiagaraPosition=$Output.RotateAroundPoint001.OutputPosition) {isSetParameters=true}
   &SkeletalMeshLocation(
    Bone / Socket Index=&Add_Integer(A=$Particles.PartitionParticles.Partition, B=4)
    Bone / Socket Sampling Mode=::NewEnumerator5
    MaskSpawnGroup=false
    Mesh Sampling Type=::NewEnumerator6
    Skeletal Mesh=$System.SkeletalMesh
   ) {module=SkeletalMeshLocation001}
   &SolveForcesAndVelocity(Clamp Velocity=false, Limit Acceleration=false)
   /SetVariables(Particles.BokehBackFill_Color:LinearColor=&MakeCustomLinearColorFromBool(Bool=$User.bIsPainted, True Color=$User.PaintedColor, False Color=<rgba 0.008 0.0284 0 5>)) {isSetParameters=true}
  ]
  rendererClasses [&NiagaraSpriteRendererProperties, &NiagaraMeshRendererProperties]
  renderers [
   {rendererClass=&NiagaraSpriteRendererProperties}
   {rendererClass=&NiagaraMeshRendererProperties, rendererIndex=1}
  ]
 }
}
summary {
 systemName=NS_DriftTrail_AstroFeather_MP1
 emitters [
  {emitterName=NE_VC_DriftTrail_Sprite_Parent, rendererClasses [&NiagaraSpriteRendererProperties]}
  {bEnabled=false, emitterName=NE_VC_DriftTrail_Ribbon_Parent003, rendererClasses [&NiagaraRibbonRendererProperties]}
  {
   emitterName=NE_VC_DriftTrail_WheelWrap_Parent001
   rendererClasses [&NiagaraSpriteRendererProperties, &NiagaraMeshRendererProperties]
  }
  {emitterName=NE_VC_DriftTrail_Ribbon_Parent001, rendererClasses [&NiagaraRibbonRendererProperties]}
 ]
 userVariables [
  User.bIsFrontendPreview {type=NiagaraBool}
  User.bIsPainted {type=NiagaraBool}
  User.ColorVariantID {type=NiagaraInt32}
  User.DarkCorePaintedScale {type=NiagaraFloat, defaultValue=3, description=""}
  User.DarkFlamesPaintedScale {type=NiagaraFloat, defaultValue=3, description=""}
  User.DarkSmokeBrightness {type=NiagaraFloat, defaultValue=1, description=""}
  User.DTCarOffset {type=NiagaraPosition, description="", defaultValue {}}
  User.FastTrailPaintedScale {type=NiagaraFloat, defaultValue=1, description=""}
  User.LightTrailPaintedScale {type=NiagaraFloat, defaultValue=1, description=""}
  User.offset {type=Vector3f, description="offset from wheels(set through cosmetic template)", defaultValue {struct=Vector3f, value=<0 0 0>}}
  User.PaintedColor {type=LinearColor, description="", defaultValue {struct=LinearColor, value=<rgba 1 1 1 1>}}
  User.PaintedScalar {type=NiagaraFloat, defaultValue=1, description=""}
  User.SlowTrailPaintedScale {type=NiagaraFloat, defaultValue=0.9, description=""}
  User.SparksPaintedScale {type=NiagaraFloat, defaultValue=4, description=""}
  User.VehicleFlags {type=NiagaraInt32, defaultValue=806289408}
  User.Wheel_Radius_Array {
   type=NiagaraDataInterfaceArrayFloat
   defaultValue {
    struct=NiagaraExt_VariableValue_DataInterface
    value {
     dataInterface=&self:NiagaraDataInterfaceArrayFloat_0
     dataInterfaceClass=&NiagaraDataInterfaceArrayFloat
    }
   }
  }
  User.Wheel_Width_Array {
   type=NiagaraDataInterfaceArrayFloat
   defaultValue {
    struct=NiagaraExt_VariableValue_DataInterface
    value {
     dataInterface=&self:NiagaraDataInterfaceArrayFloat_1
     dataInterfaceClass=&NiagaraDataInterfaceArrayFloat
    }
   }
  }
  User.WheelSize {type=NiagaraFloat, defaultValue=0.1, description=""}
 ]
}
